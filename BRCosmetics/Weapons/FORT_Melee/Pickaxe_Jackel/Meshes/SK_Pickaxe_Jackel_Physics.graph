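# Physics asset for the Jackel two-handed pickaxe skeletal mesh, defining collision and constraints for its handle and dangling "dyn_hair" bones.
# One long capsule covers the handle, and smaller capsules cover each of the four hair bones.
# Each hair bone is constrained directly to the handle rather than chained to its neighbor.
# Hair 1 and 2 carry small limits on all angular axes, while hair 3 and 4 have a nonzero limit only on swing1.
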PhysicsAsset /BRCosmetics/Weapons/FORT_Melee/Pickaxe_Jackel/Meshes/SK_Pickaxe_Jackel_Physics
bodies {
 Melee_TwoHanded_Handle [{shapeType=Capsule, length=118.3902, radius=3.0718, center=<1.558 -1.0192 19.4322>, rotation=<rot -0.0207 -89.9999 0.4607>}]
 dyn_hair_1 [{shapeType=Capsule, length=7, radius=2.25, center=<1.8729 0 0.1538>, rotation=<rot -85.304 -0.0003 0.0003>}]
 dyn_hair_2 [{shapeType=Capsule, length=11, radius=2.25, center=<3.0124 -0.6527 -0.4834>, rotation=<rot -80.8842 -179.9998 167.9247>}]
 dyn_hair_3 [{shapeType=Capsule, length=12.483, radius=7.3804, center=<1.1317 2.1655 0.8575>, rotation=<rot -89.1514 -0.0028 70.0325>}]
 dyn_hair_4 [{shapeType=Capsule, length=8.6275, radius=4.6225, center=<0.8146 -0.1568 -1.392>, rotation=<rot 0.5006 95.5401 90.0481>}]
}
constraints [
 {bone1Name=dyn_hair_4, bone2Name=Melee_TwoHanded_Handle, swing1LimitDegrees=30, swing2LimitDegrees=0, twistLimitDegrees=0}
 {bone1Name=dyn_hair_3, bone2Name=Melee_TwoHanded_Handle, swing1LimitDegrees=10, swing2LimitDegrees=0, twistLimitDegrees=0}
 {bone1Name=dyn_hair_1, bone2Name=Melee_TwoHanded_Handle, swing1LimitDegrees=10, swing2LimitDegrees=10, twistLimitDegrees=10}
 {bone1Name=dyn_hair_2, bone2Name=Melee_TwoHanded_Handle, swing1LimitDegrees=10, swing2LimitDegrees=10, twistLimitDegrees=10}
]
