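# Animation Blueprint for the Jackel pickaxe cosmetic skeletal mesh, adding looping animation and secondary physics motion.
# A single-state machine plays the pickaxe animation slightly sped up, feeding a RigidBody node that simulates its physics asset in the space of the two-handed handle bone.
# An Anim Dynamics node then springs the dyn_head bone within tight limits under reduced gravity before output.
# The event graph is empty.
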AnimBlueprint /BRCosmetics/Weapons/FORT_Melee/Pickaxe_Jackel/Meshes/SK_Pickaxe_Jackel_AnimBP
target_skeleton=&/SK_Pickaxe_Jackel_Skeleton
graphs [
 AnimGraph {nodes [
   AnimGraphNode_StateMachine_0 = New State Machine() {is=title AnimGraphNode_StateMachine}
   n1 = RigidBody(ComponentPose=Local To Component(LocalPose=@AnimGraphNode_StateMachine_0)) {
    bOverrideWorldGravity=true
    overridePhysicsAsset=&/SK_Pickaxe_Jackel_Physics
    simulationSpace=BaseBoneSpace
    alphaBoolBlend {blendOption=Linear}
    baseBoneRef {boneName=Melee_TwoHanded_Handle}
    componentLinearAccScale=<0.5 0.5 0.5>
    componentLinearVelScale=<0.5 0.5 0.5>
    overrideWorldGravity {z=-980}
   }
   n2 = Anim Dynamics - Bone: dyn_head(ComponentPose=@n1, Alpha=1, ExternalForce="0,0,0") {
    angularDampingOverride=0.999
    angularSpringConstant=2000
    bAngularSpring=true
    bGravityOverrideInSimSpace=true
    bOverrideAngularDamping=true
    bOverrideLinearDamping=true
    bUseGravityOverride=true
    gravityScale=0.5
    linearDampingOverride=0.999
    alphaBoolBlend {blendOption=Linear}
    boundBone {boneName=dyn_head}
    componentAppliedLinearAccClamp=<1000 1000 1000>
    componentLinearAccScale=<1 1 1>
    componentLinearVelScale=<1 1 1>
    gravityOverride {z=-980}
    physicsBodyDefinitions [{
      sphereCollisionRadius=0
      boundBone {boneName=dyn_head}
      boxExtents=<10 2 2>
      constraintSetup {
       angularTargetAxis=AxisY
       angularLimitsMax=<2 0 0>
       angularLimitsMin=<-2 0 0>
       angularTarget=<0 0 0>
       linearAxesMax=<7 0 0>
       linearAxesMin=<-2 0 0>
      }
      localJointOffset=<-1 0 0>
     }]
   }
   Output Pose(Result=Component To Local(ComponentPose=@n2)) {layerGroup=None}
  ]}
 New State Machine {entry=Base, path=AnimGraph.AnimGraphNode_StateMachine_0.New State Machine, ids=id, nodes [AnimStateNode_0 = Base() {is=title AnimStateNode}]}
 Base {
  path=AnimGraph.AnimGraphNode_StateMachine_0.New State Machine.AnimStateNode_0.Base
  nodes [
   n1 = Sequence Player 'Anim_Pickaxe_Jackel'() {sequence=/BRCosmetics/Weapons/FORT_Melee/Pickaxe_Jackel/Animation/Anim_Pickaxe_Jackel, playRateScaleBiasClampConstants {scale=1.3}}
   Output Animation Pose(Result=@n1)
  ]
 }
 EventGraph {links [], nodes []}
]
variables []
